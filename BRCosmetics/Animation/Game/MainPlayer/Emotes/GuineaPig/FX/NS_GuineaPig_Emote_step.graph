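# Niagara footstep effect for the GuineaPig emote, emitting a small dust or smoke puff.
# The only active emitter bursts a few flipbook sprites from a tiny ring, with time-of-day tint, gentle outward and upward velocity, drag and gravity.
# Lifetime curves drive sprite size, color, spin, sub-UV frame and material parameters.
# Disabled splash and surface-snapped foam mesh emitters remain, gated by a bIsInWater parameter and positioned through a water body data interface.
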NiagaraSystem /BRCosmetics/Animation/Game/MainPlayer/Emotes/GuineaPig/FX/NS_GuineaPig_Emote_step
dependencies {
 usedDataInterfaces [&NiagaraDataInterfaceCurve, &NiagaraDataInterfaceSpriteRendererInfo]
 usedDynamicInputs []
 usedModules [&SystemState]
}
emitters {
 FN_Core_Smoke_Sprite {
  emitterUpdateScript [
   &EmitterState()
   &SpawnBurst_Instantaneous(Spawn Count=&RandomRangeInt(Minimum=3, Maximum=4), Use Spawn Probability=true)
  ]
  particleSpawnScript [
   /SetVariables(Particles.InitSpriteRotRate=&RandomRangeFloat(Minimum=-2, Maximum=2)) {isSetParameters=true}
   &InitializeParticle(
    Color=&MultiplyLinearColor(
     Linear Color=&FNTimeOfDayColor(
      Night Color=&RandomRangeLinearColor(Minimum=<rgba 0.138 0.184 0.276 3>, Maximum=<rgba 0.0753 0.092 0.1255 4>)
      Morning Color=&RandomRangeLinearColor(Minimum=<rgba 0.276 0.2691 0.2622 1>, Maximum=<rgba 0.1255 0.1217 0.1192 2>)
      Day Color=&RandomRangeLinearColor(Minimum=<rgba 0.276 0.2691 0.2622 1>, Maximum=<rgba 0.1255 0.1217 0.1192 2>)
      Evening Color=&RandomRangeLinearColor(Minimum=<rgba 0.276 0.2691 0.2622 1>, Maximum=<rgba 0.1255 0.1217 0.1192 2>)
     )
     Scale Factor=&MakeFloat4FromVectorAndFloat("Vector (RGB)"=<1 1 1>, "Float (Alpha)"=1)
    )
    Lifetime Max=0.9
    Lifetime Min=0.7
    Lifetime Mode=::NewEnumerator1
    Sprite Rotation Angle=&RandomRangeFloat(Minimum=0, Maximum=360)
    Sprite Rotation Mode=::NewEnumerator0
    Sprite Size Mode=::NewEnumerator3
    Uniform Sprite Size=&MaskFloatBySpawnGroup(
     Masked Float=&RandomRangeFloat(Minimum=50, Maximum=35)
     Passthrough Float=&RandomRangeFloat(Minimum=245, Maximum=325)
    )
    UsePositionOffset=false
    Write Material Random=true
   )
   &TorusLocation(
    Align Torus With Axis=false
    Coordinate Space:ENiagaraCoordinateSpace=::World
    Large Radius=2
    MaskSpawnGroup=false
    Offset=<0 0 -10>
    Torus Distribution Mode:ENiagaraTorusDistributionMode=::NewEnumerator1
    Torus Mode=::NewEnumerator2
    U Position:NiagaraFloat=&ReturnNormalizedExecIndex(Normalized Index Includes One=false)
   ) {enabled=false}
   &ShapeLocation(
    MaskSpawnGroup=false
    Override Seed=false
    Ring / Disc Distribution Mode:ENiagaraSphereDistributionMode=::NewEnumerator1
    Ring Radius=2
    Rotation Coordinate Space=::World
    Shape Primitive=::NewEnumerator4
    U Position=&ReturnNormalizedExecIndex(Normalized Index Includes One=false)
    UseOwnerScale=true
   )
   &AddVelocityFromPoint(Module.Use Falloff Distance=false, Velocity Strength=&RandomRangeFloat(Minimum=100, Maximum=200))
   &AddVelocity(Coordinate Space=::World, Velocity=&RandomRangeVector(Minimum=<0 0 175>, Maximum=<0 0 200>)) {module=AddVelocity002}
   &CameraOffset(Camera Offset Amount=35)
   /SetVariables(Particles.Velocity=&Multiply_VectorByFloat(Vector=$Particles.Velocity, Float=0.1)) {isSetParameters=true}
  ]
  particleUpdateScript [
   &Drag(Drag=1.5, Use Linear Drag=true, Use Rotational Drag=true)
   &GravityForce(Gravity=<0 0 -100>)
   &SolveForcesAndVelocity(Clamp Velocity=false, Limit Acceleration=false)
   &ParticleState()
   &SpriteRotationRate(Rotation Rate=&Multiply_Float(
     A=$Particles.InitSpriteRotRate
     B=&FloatFromCurve(
      FloatCurve=DataInterface() {propertyValues {Curve {keys [
          {arriveTangent=-4.0107, interpMode=RCIM_Cubic, leaveTangent=-4.0107, tangentMode=RCTM_User, time=-0.0046, value=1.0113}
          {arriveTangent=-0.2496, interpMode=RCIM_Cubic, leaveTangent=-0.2496, tangentMode=RCTM_User, time=0.5, value=0.16}
         ]}}}
      Scale Curve=250
     )
    ))
   &ScaleColor(
    Scale Alpha=&FloatFromCurve(FloatCurve=DataInterface() {propertyValues {Curve {keys [{interpMode=RCIM_Cubic, tangentMode=RCTM_User, time=0, value=1}, <at 1 0>]}}})
    Scale RGB=&VectorFromFloat(Value=&FloatFromCurve(FloatCurve=DataInterface() {propertyValues {Curve {keys [
          {arriveTangent=5.5, interpMode=RCIM_Cubic, leaveTangent=5.5, time=0, value=0.4}
          {arriveTangent=5.5, interpMode=RCIM_Cubic, leaveTangent=5.5, time=0.05, value=3}
          {arriveTangent=-2.6916, interpMode=RCIM_Cubic, leaveTangent=-2.6916, tangentMode=RCTM_User, time=0.2, value=1.5}
          {arriveTangent=-0.625, interpMode=RCIM_Cubic, leaveTangent=-0.625, time=1, value=1}
         ]}}}))
    ScaleA=true
    ScaleRGB=true
   )
   &ScaleSpriteSize(Scale Factor=&Vector2DFromFloat(Value=&FloatFromCurve(FloatCurve=DataInterface() {propertyValues {Curve {keys [
          {arriveTangent=8.1876, interpMode=RCIM_Cubic, leaveTangent=5.3162, tangentMode=RCTM_User, time=0, value=0.15}
          {arriveTangent=1.0025, interpMode=RCIM_Cubic, leaveTangent=1.0025, tangentMode=RCTM_User, time=0.161, value=0.7059}
          {arriveTangent=0.1638, interpMode=RCIM_Cubic, leaveTangent=0.1638, tangentMode=RCTM_User, time=1.0463, value=1}
         ]}}})))
   &DynamicMaterialParameters(
    Index 0 Param 1=&Add_Float(
     A=$Particles.RandomFloat
     B=&FloatFromCurve(FloatCurve=DataInterface() {propertyValues {Curve {keys [
          {arriveTangent=2.7533, interpMode=RCIM_Cubic, leaveTangent=2.847, tangentMode=RCTM_Break, time=0, value=0}
          {arriveTangent=1.3581, interpMode=RCIM_Cubic, leaveTangent=1.3581, tangentMode=RCTM_User, time=0.2558, value=0.5125}
          {arriveTangent=0.0529, interpMode=RCIM_Cubic, leaveTangent=0.2021, tangentMode=RCTM_User, time=1, value=1}
         ]}}})
    )
    Index 0 Param 3=&FloatFromCurve(FloatCurve=DataInterface() {propertyValues {Curve {keys [{arriveTangent=-53.9199, time=0, value=0}, {leaveTangent=0.2723, time=0.25, value=0}]}}})
    Index 0 Param 4=&FloatFromCurve(FloatCurve=DataInterface() {propertyValues {Curve {keys [<at 0 1>,<at 1 0>]}}})
    Index 1 Param 1=&FloatFromCurve(FloatCurve=DataInterface() {propertyValues {Curve {keys [<at 0 0>, {arriveTangent=4.2257, interpMode=RCIM_Cubic, leaveTangent=4.2257, tangentMode=RCTM_User, time=0.25, value=0}]}}})
    Index 1 Param 2=&FloatFromCurve(FloatCurve=DataInterface() {propertyValues {Curve {keys [<at 0 0>]}}})
    Index 1 Param 3=&FloatFromCurve(FloatCurve=DataInterface() {propertyValues {Curve {keys [{interpMode=RCIM_Cubic, leaveTangent=6.3917, tangentMode=RCTM_Break, time=0, value=0}]}}})
    Index 1 Param 4=&FloatFromCurve(FloatCurve=DataInterface() {propertyValues {Curve {keys [{arriveTangent=-0.136, leaveTangent=-0.136, time=0, value=1}]}}})
    Index 2 Param 1=&FloatFromCurve(FloatCurve=DataInterface() {propertyValues {Curve {keys [
         {arriveTangent=2.646, interpMode=RCIM_Cubic, leaveTangent=2.646, tangentMode=RCTM_User, time=0, value=0}
         {leaveTangent=-0.0719, time=1, value=1}
        ]}}})
    Index 2 Param 2=&FloatFromCurve(
     FloatCurve=DataInterface() {propertyValues {Curve {keys [
         {arriveTangent=1.9056, time=0, value=0.2}
         {arriveTangent=0.0259, interpMode=RCIM_Cubic, leaveTangent=0.0259, tangentMode=RCTM_User, time=1, value=1.4}
        ]}}}
     Scale Curve=&RandomRangeFloat(Minimum=0.5, Maximum=1)
    )
    Index 2 Param 3=&FloatFromCurve(FloatCurve=DataInterface() {propertyValues {Curve {keys [
         {interpMode=RCIM_Cubic, leaveTangent=2.3064, tangentMode=RCTM_Break, time=0, value=1}
         {interpMode=RCIM_Cubic, leaveTangent=11.1233, tangentMode=RCTM_Break, time=1, value=2}
        ]}}})
    Index 2 Param 4=&FloatFromCurve(
     FloatCurve=DataInterface() {propertyValues {Curve {keys [
         {arriveTangent=-0.6867, interpMode=RCIM_Cubic, leaveTangent=-0.6867, tangentMode=RCTM_User, time=0, value=1}
         {interpMode=RCIM_Cubic, leaveTangent=-1.2956, tangentMode=RCTM_Break, time=1, value=0.75}
        ]}}}
     Scale Curve=1.6
    )
    Module.Param0WriteEnabled=true
    Module.Param0WriteEnabled1=true
    Module.Param0WriteEnabled2=true
    Module.Param1WriteEnabled=false
    Module.Param1WriteEnabled1=true
    Module.Param1WriteEnabled2=true
    Module.Param2WriteEnabled=true
    Module.Param2WriteEnabled1=true
    Module.Param2WriteEnabled2=true
    Module.Param3WriteEnabled=true
    Module.Param3WriteEnabled1=true
    Module.Param3WriteEnabled2=true
    Write Parameter Index 1=true
    Write Parameter Index 2=true
   )
   &SubUVAnimation(
    End Frame:NiagaraInt32=&Add_Integer(A=0, B=35)
    SubUV Animation Mode:ENiagara_SubUVLookupMode=::NewEnumerator1
    SubUV Curve=DataInterface() {propertyValues {Curve {keys [
        {arriveTangent=3.375, interpMode=RCIM_Cubic, leaveTangent=3.375, time=0, value=0}
        {arriveTangent=1, interpMode=RCIM_Cubic, leaveTangent=1, time=0.2, value=0.675}
        {arriveTangent=0.4062, interpMode=RCIM_Cubic, leaveTangent=0.4062, time=1, value=1}
       ]}}}
   )
  ]
  rendererClasses [&NiagaraSpriteRendererProperties]
  renderers [{rendererClass=&NiagaraSpriteRendererProperties}]
 }
 FN_Water_Foam_SnapToSurface {
  bEnabled=false
  emitterUpdateScript [
   &EmitterState()
   &SpawnBurst_Instantaneous(
    Spawn Probability=&MakeCustomFloatFromBool(Bool=$User.bIsInWater, True Float=1, False Float=0)
    Use Spawn Probability=true
    UseLoopCountLimit=false
   )
  ]
  particleSpawnScript [
   &InitializeParticle(
    Lifetime=&RandomRangeFloat(Minimum=2, Maximum=3)
    Mass Max=1.2
    Mass Min=0.6
    Mass Mode=::NewEnumerator3
    Mesh Scale Mode=::NewEnumerator3
    Mesh Uniform Scale=150
    Position Offset=<0 0 500>
    Sprite Rotation Mode=::NewEnumerator2
    UsePositionOffset=true
    Write Material Random=true
   )
   &InitialMeshOrientation(
    Mesh Orientation Mode=::NewEnumerator2
    Rotation=&RandomRangeVector(Minimum=<0 0 0>, Maximum=<0 0 1>)
    Use Rotation Vector=true
   )
   &SkeletalMeshLocation(
    Bone Sampling Mode=::NewEnumerator4
    MaskSpawnGroup=false
    Orientation Sampling=::NewEnumerator1
    Skeletal Mesh=$System.Skeletal Mesh
    Use Sampled Mesh Attribute as Ribbon ID=false
   )
   &InheritVelocity(Module.Limit Speed=true)
  ]
  particleUpdateScript [
   &ParticleState()
   &ScaleMeshSize(Scale Factor=&VectorFromFloat(Value=&FloatFromCurve(FloatCurve=DataInterface() {propertyValues {Curve {keys [{tangentMode=RCTM_User, time=0, value=0}, {tangentMode=RCTM_User, time=1, value=1.5}]}}})))
   /Game/Effects/Niagara/Modules/Water/WaterBodyLocation(WaterBodyActor:NiagaraDataInterfaceWater=$User.WaterBodyActor)
   &Drag(Drag=2, Use Linear Drag=true, Use Rotational Drag=false)
   &SolveForcesAndVelocity(Clamp Velocity=false, Limit Acceleration=false)
   &DynamicMaterialParameters(
    Index 0 Param 1=&FloatFromCurve(FloatCurve=DataInterface() {propertyValues {Curve {keys [<at -0.0046 1>,<at 1.0011 4>]}}})
    Index 0 Param 2=&FloatFromCurve(FloatCurve=DataInterface() {propertyValues {Curve {keys [<at -0.0046 0>,<at 1.0011 2>]}}})
    Index 0 Param 3=&FloatFromCurve(FloatCurve=DataInterface() {propertyValues {Curve {keys [<at -0.0046 1>,<at 1.0011 0.5>]}}})
    Index 1 Param 3=&Multiply_Float(
     A=200
     B=&Multiply_Float(A=&MakeFloatFromVector(Vector=$Particles.Scale, Channel=::NewEnumerator0), B=0.0067)
    )
    Index 2 Param 1=&FloatFromCurve(FloatCurve=DataInterface() {propertyValues {Curve {keys [<at -0.0046 0.4>,<at 1.0011 1.5>]}}})
    Index 2 Param 2=&FloatFromCurve(FloatCurve=DataInterface() {propertyValues {Curve {keys [<at -0.0046 0.1>,<at 0.11 2.6>,<at 1.0011 0.3>]}}})
    Index 2 Param 3=0.35
    Index 2 Param 4=&FloatFromCurve(FloatCurve=DataInterface() {propertyValues {Curve {keys [
         {arriveTangent=-1.9991, interpMode=RCIM_Cubic, time=0, value=1}
         {arriveTangent=-2.7143, interpMode=RCIM_Cubic, leaveTangent=-2.7143, tangentMode=RCTM_User, time=0.2095, value=0.6901}
         <at 1 0 RCIM_Cubic>
        ]}}})
    Module.Index0Param0Write=true
    Module.Index0Param1Write=true
    Module.Index0Param2Write=true
    Module.Index0Param3Write=false
    Module.Index1Param0Write=false
    Module.Index1Param1Write=false
    Module.Index1Param2Write=true
    Module.Index1Param3Write=false
    Module.Index2Param0Write=true
    Module.Index2Param1Write=true
    Module.Index2Param2Write=true
    Module.Index2Param3Write=true
    Write Parameter Index 1=true
    Write Parameter Index 2=true
   )
   &ScaleColor(ScaleA=true, ScaleRGB=true)
  ]
  rendererClasses [&NiagaraMeshRendererProperties]
  renderers [{rendererClass=&NiagaraMeshRendererProperties, rendererIndex=0}]
 }
 Splash {
  bEnabled=false
  emitterUpdateScript [
   &EmitterState()
   &SpawnBurst_Instantaneous(
    Spawn Count=35
    Spawn Probability=&MakeCustomFloatFromBool(Bool=$User.bIsInWater, True Float=1, False Float=0)
    Use Spawn Probability=true
    UseLoopCountLimit=false
   )
  ]
  particleSpawnScript [
   &InitializeParticle(
    Lifetime Max=0.65
    Lifetime Min=0.35
    Lifetime Mode=::NewEnumerator1
    Position Offset Coordinate Space=::Local
    Sprite Rotation Mode=::NewEnumerator2
    Sprite Size=&Multiply_Vector2D_ByFloat(Float=&RandomRangeFloat(Minimum=0.8, Maximum=1), Vector2D=<150 150>)
    Sprite Size Mode=::NewEnumerator4
    UsePositionOffset=true
    Write Material Random=true
   )
   &SkeletalMeshLocation(
    Bone Sampling Mode=::NewEnumerator4
    MaskSpawnGroup=false
    Orientation Sampling=::NewEnumerator1
    Skeletal Mesh=$System.Skeletal Mesh
    Use Sampled Mesh Attribute as Ribbon ID=false
   ) {enabled=false}
   &AddVelocity(
    Module.Use Falloff Distance=false
    Origin Offset=<25 0 -50>
    Override Seed=false
    Velocity Mode=::NewEnumerator1
    Velocity Speed=&Lerp_Float(A=150, B=200, Alpha=$System.WheelSpeedNormalized)
   ) {enabled=false}
   &AddVelocity(
    Cone Angle=180
    Cone Axis=<0 0 1>
    Inner Cone Angle=150
    Module.UseConeFalloff=false
    Module.UseVelDistribution=false
    Override Seed=false
    Velocity Mode=::NewEnumerator2
    Velocity Speed=&RandomRangeFloat(Minimum=200, Maximum=100)
   ) {module=AddVelocity001}
   &AddVelocity(Velocity=<0 0 300>) {module=AddVelocity002}
   &WaterBodyLocation(WaterBodyActor:NiagaraDataInterfaceWater=$User.WaterBodyActor)
   &SubUVAnimation(
    Evaluation Type=::NewEnumerator0
    "Sprite Renderer "=DataInterface() {propertyValues {SpriteRenderer=&self:Splash_4.NiagaraSpriteRendererProperties_0}}
    SubUV Animation Mode:ENiagara_SubUVLookupModeV2=::NewEnumerator2
    UseEndFrame=false
    UseStartFrame=false
   ) {module=SubUVAnimation001}
  ]
  particleUpdateScript [
   &ParticleState()
   &GravityForce(Gravity=&Multiply_VectorByFloat(Vector=<0 0 -980>, Float=1))
   &InheritSourceMovement(
    Applied Position Scale=&Multiply_VectorByFloat(
     Vector=<1 1 1>
     Float=&FloatFromCurve(FloatCurve=DataInterface() {propertyValues {Curve {keys [{interpMode=RCIM_Cubic, tangentMode=RCTM_User, time=0, value=1}, {interpMode=RCIM_Cubic, tangentMode=RCTM_User, time=1, value=0}]}}})
    )
    ApplyForce=false
    ApplyPosition=true
    ApplyPositionFalloff=false
   )
   &Drag(Drag=2, Use Linear Drag=true, Use Rotational Drag=true)
   &VortexForce(
    UseOriginOffset=false
    Vortex Force Amount=&Multiply_Float(
     A=&RandomRangeFloat(Minimum=1500, Maximum=200)
     B=&UniformAOrBFloat(A=-1, Evaluation Type=::NewEnumerator0, Recalculate Random Each Loop=true, Override Seed=false)
    )
    Vortex Origin:NiagaraPosition=$Particles.SkeletalMeshLocation.SampledPosition
   )
   &SolveForcesAndVelocity(Clamp Velocity=false, Limit Acceleration=false)
   &ScaleSpriteSize(Scale Factor=&Vector2DFromFloat(Value=&FloatFromCurve(FloatCurve=DataInterface() {propertyValues {Curve {keys [<at 0 0.2>,<at 1 1>]}}})))
   &DynamicMaterialParameters(
    Index 0 Param 2=&FloatFromCurve(FloatCurve=DataInterface() {propertyValues {Curve {keys [{interpMode=RCIM_Cubic, tangentMode=RCTM_User, time=0, value=1}, {interpMode=RCIM_Cubic, tangentMode=RCTM_User, time=0.66, value=0}]}}})
    Index 0 Param 4=&FloatFromCurve(FloatCurve=DataInterface() {propertyValues {Curve {keys [{tangentMode=RCTM_User, time=0, value=0}, {tangentMode=RCTM_User, time=1, value=1}]}}})
    Module.Param0WriteEnabled=false
    Module.Param1WriteEnabled=true
    Module.Param2WriteEnabled=false
    Module.Param3WriteEnabled=true
   )
   &Color(Color=&MakeLinearColorFromVectorAndFloat(
     "Vector (RGB)"=<1 1 1>
     "Float (Alpha)"=&FloatFromCurve(FloatCurve=DataInterface() {propertyValues {Curve {keys [<at 0 0 RCIM_Cubic>,<at 0.4 1 RCIM_Cubic>,<at 1 0>]}}})
    ))
  ]
  rendererClasses [&NiagaraSpriteRendererProperties]
  renderers [{rendererClass=&NiagaraSpriteRendererProperties}]
 }
}
summary {
 systemName=NS_GuineaPig_Emote_step
 emitters [
  {emitterName=FN_Core_Smoke_Sprite, rendererClasses [&NiagaraSpriteRendererProperties]}
  {bEnabled=false, emitterName=Splash, rendererClasses [&NiagaraSpriteRendererProperties]}
  {bEnabled=false, emitterName=FN_Water_Foam_SnapToSurface, rendererClasses [&NiagaraMeshRendererProperties]}
 ]
 userVariables [
  User.bIsInWater {type=NiagaraBool}
  User.WaterBodyActor {
   type=NiagaraDataInterfaceWater
   description=""
   defaultValue {
    struct=NiagaraExt_VariableValue_DataInterface
    value {
     dataInterface=&self:NiagaraDataInterfaceWater_0
     dataInterfaceClass=/Script/Water.NiagaraDataInterfaceWater
    }
   }
  }
 ]
}
